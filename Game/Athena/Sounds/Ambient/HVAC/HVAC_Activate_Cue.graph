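# Sound cue for the Athena HVAC activation sound, played as ambient audio when an HVAC launch activates.
# A random node picks with equal weight among three small launch-activate wave variations.
# It plays through the Athena ambient sound class with HVAC-specific attenuation and a boosted volume multiplier.
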SoundCue /Game/Athena/Sounds/Ambient/HVAC/HVAC_Activate_Cue
root=Random_0
nodes [
 {id=Random_0, children [WavePlayer_3,WavePlayer_4,WavePlayer_5], properties {weights [1,1,1]}}
 {id=WavePlayer_3, properties {soundWaveAssetPtr=&/HVAC_Launch_Activate_Small_03}}
 {id=WavePlayer_4, properties {soundWaveAssetPtr=&/HVAC_Launch_Activate_Small_02}}
 {id=WavePlayer_5, properties {soundWaveAssetPtr=&/HVAC_Launch_Activate_Small_01}}
]
properties {
 attenuationSettings=/Game/Sounds/Attenuation/Fort_HVAC_Att
 maxDistance=4600
 soundClassObject=&Fort_Ambient_Athena
 volumeMultiplier=1.559
}
