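# HVAC_Deactivate_Cue is an Athena ambient sound cue that, by its name, plays when an HVAC unit deactivates.
# A Random node picks one of four equally weighted HVAC explosion waves each time it plays.
# It uses the Athena ambient sound class, a medium-object demolition hit attenuation asset and a boosted volume multiplier.
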SoundCue /Game/Athena/Sounds/Ambient/HVAC/HVAC_Deactivate_Cue
root=Random_0
nodes [
 {id=Random_0, children [WavePlayer_0,WavePlayer_1,WavePlayer_2,WavePlayer_3], properties {weights [1,1,1,1]}}
 {id=WavePlayer_0, properties {soundWaveAssetPtr=&/HVAC_Explode_04}}
 {id=WavePlayer_1, properties {soundWaveAssetPtr=&/HVAC_Explode_01}}
 {id=WavePlayer_2, properties {soundWaveAssetPtr=&/HVAC_Explode_02}}
 {id=WavePlayer_3, properties {soundWaveAssetPtr=&/HVAC_Explode_03}}
]
properties {
 attenuationSettings=/Game/Sounds/Attenuation/Fort_Demolition_Hit_Medium_Sized_Object_Att
 maxDistance=4000
 soundClassObject=&Fort_Ambient_Athena
 volumeMultiplier=1.559
}
