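# Sound cue for the large HVAC launch activation sound in the Athena ambient audio set.
# A Random node picks one of four equally weighted wave variations each time the cue plays.
# The cue uses the large HVAC launch attenuation asset and the Athena ambient sound class, with volume and pitch multipliers raised above default.
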SoundCue /Game/Athena/Sounds/Ambient/HVAC/SC_HVAC_Launch_Activate_Large
root=Random_0
nodes [
 {id=Random_0, children [WavePlayer_3,WavePlayer_4,WavePlayer_5,WavePlayer_6], properties {weights [1,1,1,1]}}
 {id=WavePlayer_3, properties {soundWaveAssetPtr=&/HVAC_Launch_Activate_Large_01}}
 {id=WavePlayer_4, properties {soundWaveAssetPtr=&/HVAC_Launch_Activate_Large_02}}
 {id=WavePlayer_5, properties {soundWaveAssetPtr=&/HVAC_Launch_Activate_Large_03}}
 {id=WavePlayer_6, properties {soundWaveAssetPtr=&/HVAC_Launch_Activate_Large_04}}
]
properties {
 attenuationSettings=&/ATT_HVAC_Launch_Large
 maxDistance=6000
 pitchMultiplier=1.1
 soundClassObject=&Fort_Ambient_Athena
 volumeMultiplier=1.559
}
